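# Niagara cosmetic effect for the ArmyFlour back bling: a blue glowing core with light rays and lightning, apparently kill-reactive through User.KillCount.
# CoreGlow follows a socket with sprite, light and component renderers, pulsing colour and size along System Timeline time.
# Lightning spawns ambient mesh arcs, while Lightning001 and socket-partitioned LightRays spawn bursts gated by state-change and Reactive flags.
# VariantIndex selects lightning brightness; front-end modifiers retune sizes and material intensity.
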NiagaraSystem /BRCosmetics/Accessories/FORT_Backpacks/Backpack_M_MED_ArmyFlour/FX/NS_Backpack_ArmyFlour
dependencies {
 usedDataInterfaces [
  &NiagaraDataInterfaceCurve
  &NiagaraDataInterfaceColorCurve
  &NiagaraDataInterfaceArrayFloat
  &NiagaraDataInterfaceMeshRendererInfo
  &NiagaraDataInterfaceSocketReader
 ]
 usedDynamicInputs [
  &SimulationPosition
  &SelectFloatFromArray
  &Multiply_Int
  &Subtract_Vector
  &SetBoolByFloatComparison
  &SetBoolByIntComparison
 ]
 usedModules [
  &self:SystemSpawnScript.SystemScriptSource.SystemScriptGraph.NiagaraNodeAssignment_1.SetVariables_#1
  &self:SystemSpawnScript.SystemScriptSource.SystemScriptGraph.NiagaraNodeAssignment_3.SetVariables_#2
  &SystemState
  &self:SystemSpawnScript.SystemScriptSource.SystemScriptGraph.NiagaraNodeAssignment_0.SetVariables_#3
  /Game/Effects/Niagara/Modules/Logic/UpdateValueStateChange
  &Timeline
 ]
}
emitters {
 CoreGlow {
  emitterUpdateScript [
   &EmitterState(Life Cycle Mode=::NewEnumerator1, Loop Behavior=::NewEnumerator1, Loop Duration Mode=::NewEnumerator1, UseLoopDelay=false)
   &SpawnBurst_Instantaneous(Use Spawn Probability=false, UseLoopCountLimit=false)
  ]
  particleSpawnScript [&InitializeParticle(
    Color=<rgba 0.1616 0.8326 2 2>
    Position Offset=$System.Offset
    Sprite Size Mode=::NewEnumerator3
    Uniform Sprite Size=17
    UsePositionOffset=true
    Write Material Random=true
   )]
  particleUpdateScript [
   &ParticleState(Kill Particles When Lifetime Has Elapsed=false, Let Infinitely Lived Particles Die When Emitter Deactivates=true)
   &ScaleSpriteSize(
    Uniform Curve Index=$System.Timeline.Time
    Uniform Curve Sprite Scale=DataInterface() {propertyValues {Curve {keys [{tangentMode=RCTM_User, time=0, value=1}, {tangentMode=RCTM_User, time=0.2, value=1.4}, <at 0.6749 1.4>, <at 1.0025 1>]}}}
   )
   &SocketLocation(SamplePosOffset=false, Socket Reader=$System.Initial.Socket Reader, Socket Sampling Mode=::NewEnumerator2, VelocitySpeedLimit=false)
   &FN_Light_Attributes(
    Color=$Particles.Color
    Create Color Attribute=true
    Create Exponent Attribute=true
    Create Radius Attribute=true
    Exponent=3
    Radius=50
    Scale Color=3
   )
   /SetVariables(
    Particles.BigGlowColor:LinearColor=&LerpLinearColors(
     StartColor=<rgba 0.1396 0.5921 1 0.4>
     EndColor=<rgba 0.0914 0.3951 1 2>
     LerpFactor=&FloatFromCurve(
      FloatCurve=DataInterface() {propertyValues {Curve {keys [<at 0 0>,<at 0.1 1>,<at 0.6779 0.7883>,<at 1 0>]}}}
      CurveIndex=$System.Timeline.Time
     )
    )
    Particles.BigGlowSize:Vector2f=&Vector2DFromFloat(Value=&Lerp_Float(
      A=&FN_FrontEndFloatModifier(Gameplay=45, Front End=30)
      B=&FN_FrontEndFloatModifier(Gameplay=55, Front End=45)
      Alpha=&FloatFromCurve(FloatCurve=DataInterface() {propertyValues {Curve {keys [<at 0 0>,<at 0.1 1>,<at 0.4 0.3>,<at 1 0>]}}}, CurveIndex=$System.Timeline.Time)
     ))
   ) {isSetParameters=true}
   /SetVariables(
    Particles.CoreGlowColor:LinearColor=&LerpLinearColors(
     StartColor=<rgba 2.1187 3 2.899 2>
     EndColor=<rgba 5.9878 9.1332 10 2>
     LerpFactor=&FloatFromCurve(
      FloatCurve=DataInterface() {propertyValues {Curve {keys [<at 0 0>,<at 0.1 1>,<at 0.6779 0.7883>,<at 1 0>]}}}
      CurveIndex=$System.Timeline.Time
     )
    )
    Particles.CoreGlowSize:Vector2f=&Vector2DFromFloat(Value=&Lerp_Float(
      A=&FN_FrontEndFloatModifier(Gameplay=8, Front End=7)
      B=&FN_FrontEndFloatModifier(Gameplay=30, Front End=20)
      Alpha=&FloatFromCurve(FloatCurve=DataInterface() {propertyValues {Curve {keys [<at 0 0>,<at 0.1 1>,<at 0.4 0.3>,<at 1 0>]}}}, CurveIndex=$System.Timeline.Time)
     ))
   ) {isSetParameters=true}
   &ScaleColor(
    Scale Alpha=&FloatFromCurve(
     FloatCurve=DataInterface() {propertyValues {Curve {keys [<at 0 1>,<at 0.1465 10>,<at 0.4946 4.3887>,<at 1 1>]}}}
     CurveIndex=$System.Timeline.Time
    )
    ScaleA=true
    ScaleRGB=true
   )
  ]
  rendererClasses [
   &NiagaraSpriteRendererProperties
   &NiagaraLightRendererProperties
   &NiagaraComponentRendererProperties
  ]
  renderers [
   {rendererClass=&NiagaraSpriteRendererProperties}
   {rendererClass=&NiagaraLightRendererProperties, rendererIndex=1}
   {rendererClass=&NiagaraSpriteRendererProperties, rendererIndex=2}
   {rendererClass=&NiagaraComponentRendererProperties, rendererIndex=3}
   {rendererClass=&NiagaraSpriteRendererProperties, rendererIndex=4}
  ]
 }
 LightRays {
  emitterUpdateScript [
   &EmitterState()
   &SpawnBurst_Instantaneous(Spawn Count=10, Use Spawn Probability=false, UseLoopCountLimit=false) {enabled=false}
   &SpawnPerFrame(
    Spawn:NiagaraBool=&Boolean_AndOperation(
     Bool1=&Boolean_AndOperation(
      Bool1=&Boolean_NotOperation(Bool=$System.UpdateValueStateChange.Changed)
      Bool2=&Boolean_NotOperation(Bool=$Transient.FirstFrame)
     )
     Bool2=$System.Reactive
    )
    Spawn Count=11
    Use Spawn Probability=false
   )
  ]
  particleSpawnScript [
   &PartitionParticles(Number Of Partitions=11)
   &InitializeParticle(
    Color Mode=::NewEnumerator0
    Lifetime=&RandomRangeFloat(Minimum=0.85, Maximum=1.1)
    Mesh Scale=&FN_FrontEndVector3Modifier(Gameplay=<0.45 0.45 0.9>, Front End=<0.45 0.45 0.8>)
    Mesh Scale Mode=::NewEnumerator4
    Sprite Size Mode=::NewEnumerator3
    Uniform Sprite Size=12
    UsePositionOffset=false
    Write Material Random=true
   )
   &InitialMeshOrientation(Mesh Orientation Mode=::NewEnumerator2, Orientation Vector=<0 0 1>, Use Rotation Vector=false) {enabled=false}
  ]
  particleUpdateScript [
   &ParticleState()
   &ScaleMeshSize(Scale Factor=&MakeVector(
     X=1
     Y=1
     Z=&FloatFromCurve(FloatCurve=DataInterface() {propertyValues {Curve {keys [<at 0.0011 0.8604>, {interpMode=RCIM_Cubic, leaveTangent=-1.083, tangentMode=RCTM_Break, time=0.1951, value=1.2}, <at 0.45 1>]}}})
    ))
   &ScaleColor(
    Linear Color Curve=DataInterface() {propertyValues {
      AlphaCurve {keys [<at 0.2165 1>,<at 0.4529 0.3>,<at 1 0>]}
      BlueCurve {keys [<at 0.046 3>,<at 0.2253 1.2>,<at 0.9976 1.0753>]}
      GreenCurve {keys [<at 0.046 2.8219>,<at 0.2253 0.6681>,<at 0.9976 0.3122>]}
      RedCurve {keys [<at 0.046 1.3802>,<at 0.2253 0.0969>,<at 0.9976 0.1209>]}
     }}
    Scale Mode=::NewEnumerator2
   )
   &SocketLocation(
    SamplePosOffset=false
    Socket Index=&Add_Integer(A=$Particles.PartitionParticles.Partition, B=1)
    Socket Reader=$System.Socket Reader
    Socket Sampling Mode=::NewEnumerator2
    VelocitySpeedLimit=false
   )
   &DynamicMaterialParameters(
    Index 0 Param 1=&RandomRangeFloat(Minimum=0, Maximum=1)
    Index 0 Param 2=1
    Index 0 Param 3=1
    Module.Index0Float0Write=true
    Module.Index0Float1Write=true
    Module.Index0Float2Write=true
    Module.Index0Float3Write=true
   )
   /SetVariables(Particles.GlowCameraOffset=4) {isSetParameters=true}
  ]
  rendererClasses [&NiagaraMeshRendererProperties, &NiagaraSpriteRendererProperties]
  renderers [
   {rendererClass=&NiagaraMeshRendererProperties, rendererIndex=0}
   {rendererClass=&NiagaraSpriteRendererProperties, rendererIndex=1}
  ]
 }
 Lightning {
  emitterUpdateScript [
   &EmitterState()
   &SpawnRate(
    SpawnRate=&FloatFromCurve(
     FloatCurve=DataInterface() {propertyValues {Curve {
        defaultValue=1
        keys [{interpMode=RCIM_Cubic, tangentMode=RCTM_User, time=0, value=0.3}, {interpMode=RCIM_Cubic, tangentMode=RCTM_User, time=0.7999, value=0.99}]
       }}}
     CurveIndex=$Emitter.NormalizedLoopAge
     Scale Curve=5
    )
    Use Spawn Probability=true
   )
  ]
  particleSpawnScript [
   /SetVariables(Particles.SpawnGroup=$Emitter.SpawnGroup) {isSetParameters=true, module=SetVariables_#4}
   &InitializeParticle(
    Color=<rgba 0.0485 0.5987 2 1>
    Lifetime=&RandomRangeFloat(Minimum=0.25, Maximum=0.75)
    Mesh Scale Mode=::NewEnumerator3
    Mesh Uniform Scale=0.5
    UsePositionOffset=true
    Write Material Random=true
   )
   /SetVariables(Particles.bFlipMesh:NiagaraBool=&RandomBool(Randomness Mode=::NewEnumerator2, Random Seed=?)) {isSetParameters=true, module=SetVariables_#5}
   /SetVariables(Particles.MeshIndex=&RandomRangeInt(Minimum=0, Maximum=2)) {isSetParameters=true, module=SetVariables_#6}
   /SetVariables(Particles.Scale=&Multiply_VectorByFloat(
     Vector=&RandomRangeVector(Minimum=<1 1 0.5>, Maximum=<1 1 0.75>)
     Float=&RandomRangeFloat(Minimum=0.2, Maximum=0.2)
    )) {isSetParameters=true, module=SetVariables_#7}
   &ShapeLocation(
    Cylinder Distribution=::NewEnumerator2
    Cylinder Height=1
    Cylinder Radius=7
    MaskSpawnGroup=false
    Offset=$System.CenterOffset
    Offset Mode=::NewEnumerator0
    Rotation Angle Type:ENiagara_AngleInput=::NewEnumerator0
    Rotation Axis=<0 1 0>
    Rotation Mode=::NewEnumerator1
    Shape Primitive=::NewEnumerator1
    U Position=&RandomRangeFloat(Minimum=0, Maximum=1)
    UseOwnerScale=true
    V Position=&RandomRangeFloat(Minimum=0, Maximum=1)
   ) {enabled=false}
   &InitialMeshOrientation(
    Rotation=&MakeVector(
     X=0
     Y=0
     Z=&MakeCustomFloatFromBool(Bool=&RandomBool(), True Float=1, False Float=0)
    )
    Use Rotation Vector=true
   )
  ]
  particleUpdateScript [
   &DynamicMaterialParameters(
    Index 0 Param 1=&FN_FrontEndFloatModifier(Gameplay=2, Front End=2.5)
    Index 0 Param 2=&FloatFromCurve(
     FloatCurve=DataInterface() {propertyValues {Curve {
        defaultValue=1
        keys [
         {arriveTangent=9.0725, interpMode=RCIM_Cubic, leaveTangent=9.0725, tangentMode=RCTM_User, time=0, value=0}
         {arriveTangent=-2.5249, interpMode=RCIM_Cubic, leaveTangent=-2.5249, tangentMode=RCTM_User, time=0.1453, value=1.1569}
         {arriveTangent=-2.6402, interpMode=RCIM_Cubic, leaveTangent=-2.6402, time=0.289, value=0.0335}
         {arriveTangent=-2.1502, interpMode=RCIM_Cubic, leaveTangent=-2.1502, tangentMode=RCTM_User, time=0.4576, value=0.3325}
         {arriveTangent=-0.0405, interpMode=RCIM_Cubic, leaveTangent=-0.0405, tangentMode=RCTM_User, time=1, value=0}
        ]
       }}}
     Scale Curve=0.8
    )
    Index 0 Param 3=&Add_Float(
     A=&RandomRangeFloat(Minimum=0, Maximum=10)
     B=&Multiply_Float(
      A=&MakeCustomFloatFromBool(Bool=&RandomBool(), True Float=1, False Float=-1)
      B=&Multiply_Float(A=$Particles.NormalizedAge, B=0.2)
     )
    )
    Index 1 Param 1=0.02
    Index 1 Param 3=&FloatFromCurve(FloatCurve=DataInterface() {propertyValues {Curve {keys [<at 0 1>,<at 1 0>]}}})
    Module.Index0Param0Write=true
    Module.Index0Param1Write=true
    Module.Index0Param2Write=true
    Module.Index0Param3Write=false
    Module.Index1Param0Write=true
    Module.Index1Param1Write=true
    Module.Index1Param2Write=true
    Module.Index1Param3Write=false
    Module.Index2Param0Write=true
    Module.Index2Param1Write=true
    Module.Index2Param2Write=true
    Module.Index2Param3Write=true
    Write Parameter Index 1=true
    Write Parameter Index 2=true
   )
   &ParticleState()
   &ScaleMeshSize(Scale Factor=&MakeVector(
     X=1
     Y=1
     Z=&FloatFromCurve(
      FloatCurve=DataInterface() {propertyValues {Curve {
         defaultValue=1
         keys [
          {arriveTangent=4.1785, interpMode=RCIM_Cubic, tangentMode=RCTM_Break, time=0, value=0.5}
          {interpMode=RCIM_Cubic, leaveTangent=0.0224, tangentMode=RCTM_Break, time=1, value=1}
         ]
        }}}
      Scale Curve=0.75
     )
    )) {enabled=false}
   &ScaleColor(
    Scale Alpha=&FloatFromCurve(FloatCurve=DataInterface() {propertyValues {Curve {keys [
         {arriveTangent=9.0725, interpMode=RCIM_Cubic, leaveTangent=9.0725, tangentMode=RCTM_User, time=0, value=0}
         {arriveTangent=-2.5249, interpMode=RCIM_Cubic, leaveTangent=-0.2678, tangentMode=RCTM_Break, time=0.1, value=1}
         {arriveTangent=-1.6791, interpMode=RCIM_Cubic, leaveTangent=-0.0405, tangentMode=RCTM_Break, time=1, value=0}
        ]}}})
    Scale RGB=&VectorFromFloat(Value=&FloatFromCurve(
      FloatCurve=DataInterface() {propertyValues {Curve {
         defaultValue=1
         keys [
          {arriveTangent=9.0725, interpMode=RCIM_Cubic, leaveTangent=9.0725, tangentMode=RCTM_User, time=0, value=1}
          {arriveTangent=-2.5249, interpMode=RCIM_Cubic, leaveTangent=-2.5249, tangentMode=RCTM_User, time=0.0544, value=2.9665}
         ]
        }}}
      Scale Curve=&MergeHelper_FloatSwitch(0=1.5, 1=2, 2=1.25, 3=1, 4=5, Index=$User.VariantIndex, Num=::NewEnumerator3)
     ))
    ScaleA=true
    ScaleRGB=true
   )
   &CameraOffset(Camera Offset Amount=10) {enabled=false}
  ]
  rendererClasses [&NiagaraMeshRendererProperties]
  renderers [{rendererClass=&NiagaraMeshRendererProperties, rendererIndex=0}]
 }
 Lightning001 {
  emitterUpdateScript [
   &EmitterState()
   &SpawnPerFrame(
    Spawn:NiagaraBool=&Boolean_AndOperation(
     Bool1=&Boolean_AndOperation(
      Bool1=&Boolean_NotOperation(Bool=$System.UpdateValueStateChange.Changed)
      Bool2=&Boolean_NotOperation(Bool=$Transient.FirstFrame)
     )
     Bool2=$System.Reactive
    )
    Spawn Count:NiagaraInt32=&RandomRangeInt(Minimum=4, Maximum=6)
    Use Spawn Probability=false
   )
  ]
  particleSpawnScript [
   /SetVariables(Particles.SpawnGroup=$Emitter.SpawnGroup) {isSetParameters=true, module=SetVariables_#4}
   &InitializeParticle(
    Color=<rgba 0.0727 0.8981 3 1>
    Lifetime=&RandomRangeFloat(Minimum=0.45, Maximum=0.8)
    Mesh Scale Mode=::NewEnumerator3
    Mesh Uniform Scale=2
    UsePositionOffset=true
    Write Material Random=true
   )
   /SetVariables(Particles.bFlipMesh:NiagaraBool=&RandomBool(Randomness Mode=::NewEnumerator2, Random Seed=?)) {isSetParameters=true, module=SetVariables_#5}
   /SetVariables(Particles.MeshIndex=&RandomRangeInt(Minimum=0, Maximum=2)) {isSetParameters=true, module=SetVariables_#6}
   /SetVariables(Particles.Scale=&Multiply_VectorByFloat(
     Vector=&RandomRangeVector(Minimum=<1 1 0.5>, Maximum=<1 1 0.75>)
     Float=&RandomRangeFloat(Minimum=0.4, Maximum=0.55)
    )) {isSetParameters=true, module=SetVariables_#7}
   &ShapeLocation(
    Cylinder Distribution=::NewEnumerator2
    Cylinder Height=1
    Cylinder Radius=7
    MaskSpawnGroup=false
    Offset=$System.CenterOffset
    Offset Mode=::NewEnumerator0
    Rotation Angle Type:ENiagara_AngleInput=::NewEnumerator0
    Rotation Axis=<0 1 0>
    Rotation Mode=::NewEnumerator1
    Shape Primitive=::NewEnumerator1
    U Position=&RandomRangeFloat(Minimum=0, Maximum=1)
    UseOwnerScale=true
    V Position=&RandomRangeFloat(Minimum=0, Maximum=1)
   ) {enabled=false}
   &InitialMeshOrientation(
    Rotation=&MakeVector(
     X=0
     Y=0
     Z=&MakeCustomFloatFromBool(Bool=&RandomBool(), True Float=1, False Float=0)
    )
    Use Rotation Vector=true
   )
  ]
  particleUpdateScript [
   &DynamicMaterialParameters(
    Index 0 Param 1=&FN_FrontEndFloatModifier(Gameplay=5, Front End=4)
    Index 0 Param 2=&FloatFromCurve(
     FloatCurve=DataInterface() {propertyValues {Curve {
        defaultValue=1
        keys [
         {arriveTangent=9.0725, interpMode=RCIM_Cubic, leaveTangent=9.0725, tangentMode=RCTM_User, time=0, value=0}
         {arriveTangent=-2.5249, interpMode=RCIM_Cubic, leaveTangent=-2.5249, tangentMode=RCTM_User, time=0.1455, value=1.1569}
         {arriveTangent=-2.6414, interpMode=RCIM_Cubic, leaveTangent=-2.6414, time=0.289, value=0.0335}
         {arriveTangent=-2.1502, interpMode=RCIM_Cubic, leaveTangent=-2.1502, tangentMode=RCTM_User, time=0.4576, value=0.3325}
         {arriveTangent=-0.0405, interpMode=RCIM_Cubic, leaveTangent=-0.0405, tangentMode=RCTM_User, time=1, value=0}
        ]
       }}}
     Scale Curve=0.75
    )
    Index 0 Param 3=&Add_Float(
     A=&RandomRangeFloat(Minimum=0, Maximum=10)
     B=&Multiply_Float(
      A=&MakeCustomFloatFromBool(Bool=&RandomBool(), True Float=1, False Float=-1)
      B=&Multiply_Float(A=$Particles.NormalizedAge, B=0.2)
     )
    )
    Index 1 Param 1=0.0025
    Index 1 Param 3=&FloatFromCurve(FloatCurve=DataInterface() {propertyValues {Curve {keys [<at 0 1>,<at 1 0>]}}})
    Module.Index0Param0Write=true
    Module.Index0Param1Write=true
    Module.Index0Param2Write=true
    Module.Index0Param3Write=false
    Module.Index1Param0Write=true
    Module.Index1Param1Write=true
    Module.Index1Param2Write=true
    Module.Index1Param3Write=false
    Module.Index2Param0Write=true
    Module.Index2Param1Write=true
    Module.Index2Param2Write=true
    Module.Index2Param3Write=true
    Write Parameter Index 1=true
    Write Parameter Index 2=true
   )
   &ParticleState()
   &ScaleMeshSize(Scale Factor=&MakeVector(
     X=1
     Y=1
     Z=&FloatFromCurve(
      FloatCurve=DataInterface() {propertyValues {Curve {
         defaultValue=1
         keys [
          {arriveTangent=4.1785, interpMode=RCIM_Cubic, tangentMode=RCTM_Break, time=0, value=0.5}
          {interpMode=RCIM_Cubic, leaveTangent=0.0224, tangentMode=RCTM_Break, time=1, value=1}
         ]
        }}}
      Scale Curve=0.75
     )
    )) {enabled=false}
   &ScaleColor(
    Scale Alpha=&FloatFromCurve(FloatCurve=DataInterface() {propertyValues {Curve {keys [
         {arriveTangent=9.0725, interpMode=RCIM_Cubic, leaveTangent=9.0725, tangentMode=RCTM_User, time=0, value=0}
         {arriveTangent=-2.5249, interpMode=RCIM_Cubic, leaveTangent=-0.2678, tangentMode=RCTM_Break, time=0.1, value=1}
         {arriveTangent=-1.6791, interpMode=RCIM_Cubic, leaveTangent=-0.0405, tangentMode=RCTM_Break, time=1, value=0}
        ]}}})
    Scale RGB=&VectorFromFloat(Value=&FloatFromCurve(
      FloatCurve=DataInterface() {propertyValues {Curve {
         defaultValue=1
         keys [
          {arriveTangent=9.0725, interpMode=RCIM_Cubic, leaveTangent=9.0725, tangentMode=RCTM_User, time=0, value=1}
          {arriveTangent=-2.5249, interpMode=RCIM_Cubic, leaveTangent=-2.5249, tangentMode=RCTM_User, time=0.0544, value=2.9665}
         ]
        }}}
      Scale Curve=&MergeHelper_FloatSwitch(0=1.5, 1=2, 2=1.25, 3=1, 4=5, Index=$User.VariantIndex, Num=::NewEnumerator3)
     ))
    ScaleA=true
    ScaleRGB=true
   )
   &CameraOffset(Camera Offset Amount=6)
  ]
  rendererClasses [&NiagaraMeshRendererProperties]
  renderers [{rendererClass=&NiagaraMeshRendererProperties, rendererIndex=0}]
 }
}
summary {
 systemName=NS_Backpack_ArmyFlour
 emitters [
  {
   emitterName=CoreGlow
   rendererClasses [
    &NiagaraSpriteRendererProperties
    &NiagaraLightRendererProperties
    &NiagaraComponentRendererProperties
   ]
  }
  {emitterName=LightRays, rendererClasses [&NiagaraMeshRendererProperties, &NiagaraSpriteRendererProperties]}
  {emitterName=Lightning, rendererClasses [&NiagaraMeshRendererProperties]}
  {emitterName=Lightning001, rendererClasses [&NiagaraMeshRendererProperties]}
 ]
 userVariables [
  User.bIsFrontEnd {type=NiagaraBool}
  User.bIsFrontEndPreview {type=NiagaraBool}
  User.IsReactive {type=NiagaraFloat, defaultValue=1, description=""}
  User.KillCount {type=NiagaraInt32}
  User.VariantIndex {type=NiagaraInt32}
 ]
}
